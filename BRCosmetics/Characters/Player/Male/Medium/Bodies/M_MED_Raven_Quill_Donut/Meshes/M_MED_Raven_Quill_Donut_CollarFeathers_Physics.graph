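# Physics asset for the collar feathers of the male medium Raven Quill Donut player cosmetic body.
# Three small collar bodies (left, middle, right) act as parents for eleven box-shaped feather bodies offset along their bones.
# Each feather is constrained only to its own collar body with tight swing and very narrow twist limits, so feathers move independently rather than in chains.
# Two of the left feathers have tighter swing limits than the rest.
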PhysicsAsset /BRCosmetics/Characters/Player/Male/Medium/Bodies/M_MED_Raven_Quill_Donut/Meshes/M_MED_Raven_Quill_Donut_CollarFeathers_Physics
bodies {
 dyn_collar_l [{shapeType=Box, extentX=0.5, extentY=0.5, extentZ=1.4521}]
 dyn_collar_mid [{shapeType=Box, extentX=0.5, extentY=0.5, extentZ=1.4521}]
 dyn_collar_r [{shapeType=Box, extentX=0.5, extentY=0.5, extentZ=1.4521}]
 dyn_feather_l_01 [{shapeType=Box, extentX=9.6499, extentY=5.7879, extentZ=2.096, center=<4.7278 0 0>}]
 dyn_feather_l_02 [{shapeType=Box, extentX=8.8202, extentY=5.4821, extentZ=1.6018, center=<3.7264 0 0>}]
 dyn_feather_l_03 [{shapeType=Box, extentX=9.4644, extentY=4.9777, extentZ=1.6827, center=<2.8895 0 0>}]
 dyn_feather_mid_01 [{shapeType=Box, extentX=7.756, extentY=6.9909, extentZ=2.4191, center=<2.9221 0 0>}]
 dyn_feather_mid_02 [{shapeType=Box, extentX=8.4852, extentY=4.7463, extentZ=1.6596, center=<2.952 0 0>}]
 dyn_feather_mid_03 [{shapeType=Box, extentX=11.1581, extentY=4.215, extentZ=1.4791, center=<3.9044 0 0>}]
 dyn_feather_mid_04 [{shapeType=Box, extentX=10.2575, extentY=6.1877, extentZ=1.5815, center=<3.7466 0 0>}]
 dyn_feather_r_01 [{shapeType=Box, extentX=8.104, extentY=3.6218, extentZ=1.243, center=<2.8396 0 0>}]
 dyn_feather_r_02 [{shapeType=Box, extentX=8.4744, extentY=4.4467, extentZ=1.3837, center=<3.5293 0 0>}]
 dyn_feather_r_03 [{shapeType=Box, extentX=7.8947, extentY=3.721, extentZ=1.2566, center=<2.7941 0 0>}]
 dyn_feather_r_04 [{shapeType=Box, extentX=8.2381, extentY=3.7581, extentZ=1.0409, center=<2.9375 0 0>}]
}
constraints [
 {bone1Name=dyn_feather_r_01, bone2Name=dyn_collar_r, swing1LimitDegrees=20, swing2LimitDegrees=20, twistLimitDegrees=3}
 {bone1Name=dyn_feather_r_02, bone2Name=dyn_collar_r, swing1LimitDegrees=20, swing2LimitDegrees=20, twistLimitDegrees=3}
 {bone1Name=dyn_feather_r_03, bone2Name=dyn_collar_r, swing1LimitDegrees=20, swing2LimitDegrees=20, twistLimitDegrees=3}
 {bone1Name=dyn_feather_r_04, bone2Name=dyn_collar_r, swing1LimitDegrees=20, swing2LimitDegrees=20, twistLimitDegrees=3}
 {bone1Name=dyn_feather_mid_01, bone2Name=dyn_collar_mid, swing1LimitDegrees=20, swing2LimitDegrees=20, twistLimitDegrees=3}
 {bone1Name=dyn_feather_mid_02, bone2Name=dyn_collar_mid, swing1LimitDegrees=20, swing2LimitDegrees=20, twistLimitDegrees=3}
 {bone1Name=dyn_feather_mid_03, bone2Name=dyn_collar_mid, swing1LimitDegrees=20, swing2LimitDegrees=20, twistLimitDegrees=3}
 {bone1Name=dyn_feather_mid_04, bone2Name=dyn_collar_mid, swing1LimitDegrees=20, swing2LimitDegrees=20, twistLimitDegrees=3}
 {bone1Name=dyn_feather_l_01, bone2Name=dyn_collar_l, swing1LimitDegrees=20, swing2LimitDegrees=20, twistLimitDegrees=3}
 {bone1Name=dyn_feather_l_02, bone2Name=dyn_collar_l, swing1LimitDegrees=10, swing2LimitDegrees=10, twistLimitDegrees=3}
 {bone1Name=dyn_feather_l_03, bone2Name=dyn_collar_l, swing1LimitDegrees=10, swing2LimitDegrees=10, twistLimitDegrees=3}
]
