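# Secondary-motion AnimBlueprint for the Raven Quill Donut character body, layered on a pose copied from the attached parent mesh.
# The collar bends when targeting or when the head tilts back, measured against spine_05.
# Weighted multi-bone constraints position feathers and shoulder pieces around RigidBody and AnimDynamics simulations, which also drive the scarf.
# Additive right-skirt offsets follow stationary crouch, crouch-walk and moving-jump alphas computed in the update event.
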AnimBlueprint /BRCosmetics/Characters/Player/Male/Medium/Bodies/M_MED_Raven_Quill_Donut/Meshes/M_MED_Raven_Quill_Donut_AnimBp
target_skeleton=&/M_MED_Raven_Quill_Donut_Skeleton
graphs [
 AnimGraph {nodes [
   "Transform (Modify) Bone - Bone: dyn_skirt_r_front_base"(Alpha=1, Translation="0,1,-1", Rotation="10,10,0", Scale="1,1,1") {
    rotationMode=BMM_Additive
    rotationSpace=BCS_BoneSpace
    translationMode=BMM_Additive
    translationSpace=BCS_BoneSpace
    alphaBoolBlend {blendOption=Linear}
    alphaScaleBiasClamp {bInterpResult=true, interpSpeedDecreasing=5, interpSpeedIncreasing=20}
    boneToModify {boneName=dyn_skirt_r_front_base}
    rotation {pitch=10, yaw=10}
    translation {y=1, z=-1}
   }
   n1 = Get Crouch_Alpha() -> (Crouch_Alpha:double)
   n2 = "To Float (Boolean)"(InBool=Get bIsTargeting())
   n3 = "Transform (Modify) Bone - Bone: dyn_collar_mid"!(ComponentPose=Local To Component(LocalPose=Use cached pose 'ADSCache'()), Alpha=@n2, Translation="0,0,0", Rotation="0,0,-25") -> (Pose) {
    rotationMode=BMM_Additive
    rotationSpace=BCS_BoneSpace
    scaleSpace=BCS_BoneSpace
    alphaBoolBlend {blendOption=Linear}
    alphaScaleBiasClamp {bInterpResult=true, interpSpeedDecreasing=5, interpSpeedIncreasing=6}
    boneToModify {boneName=dyn_collar_mid}
    rotation {roll=-25}
    scale=<2 2 2>
   }
   n4 = "Transform (Modify) Bone - Bone: dyn_collar_r"!(ComponentPose=@n3, Alpha=@n2, Translation="0,0,0", Rotation="15,0,25") -> (Pose) {
    rotationMode=BMM_Additive
    rotationSpace=BCS_BoneSpace
    scaleSpace=BCS_BoneSpace
    alphaBoolBlend {blendOption=Linear}
    alphaScaleBiasClamp {bInterpResult=true, interpSpeedDecreasing=5, interpSpeedIncreasing=6}
    boneToModify {boneName=dyn_collar_r}
    rotation {pitch=15, roll=25}
    scale=<2 2 2>
   }
   n5 = "Transform (Modify) Bone - Bone: dyn_collar_l"!(ComponentPose=@n4, Alpha=@n2, Translation="0,0,0", Rotation="30,0,0") -> (Pose) {
    rotationMode=BMM_Additive
    rotationSpace=BCS_BoneSpace
    scaleSpace=BCS_BoneSpace
    alphaBoolBlend {blendOption=Linear}
    alphaScaleBiasClamp {bInterpResult=true, interpSpeedDecreasing=5, interpSpeedIncreasing=6}
    boneToModify {boneName=dyn_collar_l}
    rotation {pitch=30}
    scale=<2 2 2>
   }
   n6 = Get Head_Rot_Back() -> (Head_Rot_Back:double)
   n7 = "Transform (Modify) Bone - Bone: dyn_collar_r"(ComponentPose=@n5, Alpha=@n6, Translation="0,0,0", Rotation="-25,0,0") {
    rotationMode=BMM_Additive
    rotationSpace=BCS_BoneSpace
    scaleSpace=BCS_BoneSpace
    alphaBoolBlend {blendOption=Linear}
    alphaScaleBiasClamp {bInterpResult=true, interpSpeedDecreasing=5, interpSpeedIncreasing=6}
    boneToModify {boneName=dyn_collar_r}
    rotation {pitch=-25}
    scale=<2 2 2>
   }
   n8 = "Transform (Modify) Bone - Bone: dyn_collar_l"(ComponentPose=@n7, Alpha=@n6, Translation="0,0,0", Rotation="-25,0,0") {
    rotationMode=BMM_Additive
    rotationSpace=BCS_BoneSpace
    scaleSpace=BCS_BoneSpace
    alphaBoolBlend {blendOption=Linear}
    alphaScaleBiasClamp {bInterpResult=true, interpSpeedDecreasing=5, interpSpeedIncreasing=6}
    boneToModify {boneName=dyn_collar_l}
    rotation {pitch=-25}
    scale=<2 2 2>
   }
   n9 = "Transform (Modify) Bone - Bone: dyn_collar_mid"(ComponentPose=@n8, Alpha=@n6, Translation="0,0,0", Rotation="-70,0,-30") {
    rotationMode=BMM_Additive
    rotationSpace=BCS_BoneSpace
    scaleSpace=BCS_BoneSpace
    alphaBoolBlend {blendOption=Linear}
    alphaScaleBiasClamp {bInterpResult=true, interpSpeedDecreasing=5, interpSpeedIncreasing=6}
    boneToModify {boneName=dyn_collar_mid}
    rotation {pitch=-70, roll=-30}
    scale=<2 2 2>
   }
   n10 = "Constraint - dyn_feather_l_03 (PPPPPPPPP)"!(
    ComponentPose=@n9
    Alpha=1
    ConstraintWeights_0=0.0007
    ConstraintWeights_1:float=0.0092
    ConstraintWeights_2:float=0.3825
    ConstraintWeights_3:float=0.0013
    ConstraintWeights_4:float=0.0002
    ConstraintWeights_5:float=0.0324
    ConstraintWeights_6:float=0.5619
    ConstraintWeights_7:float=0.0082
    ConstraintWeights_8:float=0.0035
   ) -> (Pose) {
    alphaBoolBlend {blendOption=Linear}
    boneToModify {boneName=dyn_feather_l_03}
    constraintSetup [
     {targetBone {boneName=head}}
     {targetBone {boneName=clavicle_l}}
     {targetBone {boneName=dyn_collar_mid}}
     {targetBone {boneName=spine_03}}
     {targetBone {boneName=spine_04}}
     {targetBone {boneName=spine_05}}
     {targetBone {boneName=dyn_collar_l}}
     {targetBone {boneName=neck_02}}
     {targetBone {boneName=neck_01}}
    ]
    constraintWeights [1,1,1,1,1,1,1,1,1]
   }
   n11 = "Constraint - dyn_feather_mid_01 (PPPPPPPPP)"!(
    ComponentPose=@n10
    Alpha=1
    ConstraintWeights_0=0.0016
    ConstraintWeights_1:float=0.0123
    ConstraintWeights_2:float=0.6295
    ConstraintWeights_3:float=0.0023
    ConstraintWeights_4:float=0.0003
    ConstraintWeights_5:float=0.0529
    ConstraintWeights_6:float=0.2782
    ConstraintWeights_7:float=0.0158
    ConstraintWeights_8:float=0.0069
   ) -> (Pose) {
    alphaBoolBlend {blendOption=Linear}
    boneToModify {boneName=dyn_feather_mid_01}
    constraintSetup [
     {targetBone {boneName=head}}
     {targetBone {boneName=clavicle_l}}
     {targetBone {boneName=dyn_collar_mid}}
     {targetBone {boneName=spine_03}}
     {targetBone {boneName=spine_04}}
     {targetBone {boneName=spine_05}}
     {targetBone {boneName=dyn_collar_l}}
     {targetBone {boneName=neck_02}}
     {targetBone {boneName=neck_01}}
    ]
    constraintWeights [1,1,1,1,1,1,1,1,1]
   }
   n12 = "Constraint - dyn_feather_mid_04 (PPPPPPP)"!(
    ComponentPose=@n11
    Alpha=1
    ConstraintWeights_0=0.8848
    ConstraintWeights_1:float=0.0709
    ConstraintWeights_2:float=0.0018
    ConstraintWeights_3:float=0.0253
    ConstraintWeights_4:float=0.0071
    ConstraintWeights_5:float=0.007
    ConstraintWeights_6:float=0.0032
   ) -> (Pose) {
    alphaBoolBlend {blendOption=Linear}
    boneToModify {boneName=dyn_feather_mid_04}
    constraintSetup [
     {targetBone {boneName=dyn_collar_mid}}
     {targetBone {boneName=dyn_collar_r}}
     {targetBone {boneName=spine_03}}
     {targetBone {boneName=spine_05}}
     {targetBone {boneName=clavicle_r}}
     {targetBone {boneName=neck_02}}
     {targetBone {boneName=neck_01}}
    ]
    constraintWeights [1,1,1,1,1,1,1]
   }
   n13 = RigidBody(ComponentPose=@n12) {
    bClampLinearTranslationLimitToRefPose=true
    bOverrideWorldGravity=true
    overridePhysicsAsset=&/M_MED_Raven_Quill_Donut_CollarFeathers_Physics
    simulationSpace=BaseBoneSpace
    alphaBoolBlend {blendOption=Linear}
    baseBoneRef {boneName=spine_03}
    componentAppliedLinearAccClamp=<3000 3000 3000>
    componentLinearAccScale=<0.25 0.25 0.25>
    componentLinearVelScale=<0.25 0.25 0.25>
    overrideWorldGravity {z=-500}
   }
   n14 = "Constraint - dyn_feather_r_04 (PPPPPPPPP)"!(
    ComponentPose=@n13
    Alpha=1
    ConstraintWeights_0=0.0001
    ConstraintWeights_1:float=0.4515
    ConstraintWeights_2:float=0.5058
    ConstraintWeights_3:float=0.0015
    ConstraintWeights_4:float=0.0002
    ConstraintWeights_5:float=0.0227
    ConstraintWeights_6:float=0.0092
    ConstraintWeights_7:float=0.0062
    ConstraintWeights_8:float=0.0027
   ) -> (Pose) {
    alphaBoolBlend {blendOption=Linear}
    boneToModify {boneName=dyn_feather_r_04}
    constraintSetup [
     {targetBone {boneName=head}}
     {targetBone {boneName=dyn_collar_r}}
     {targetBone {boneName=dyn_collar_mid}}
     {targetBone {boneName=spine_03}}
     {targetBone {boneName=spine_04}}
     {targetBone {boneName=spine_05}}
     {targetBone {boneName=clavicle_r}}
     {targetBone {boneName=neck_02}}
     {targetBone {boneName=neck_01}}
    ]
    constraintWeights [1,1,1,1,1,1,1,1,1]
   }
   n15 = RigidBody!(ComponentPose=@n14, Alpha, ExternalForce, bTransferBoneVelocities:bool=False) -> (Pose) {
    bClampLinearTranslationLimitToRefPose=true
    bOverrideWorldGravity=true
    bTransferBoneVelocities=true
    overridePhysicsAsset=/Game/Characters/Player/Male/Medium/Bodies/M_MED_Quill/Meshes/RavenQuill_02_Rigids/M_MED_Quill_02_Physics
    simulationSpace=BaseBoneSpace
    alphaBoolBlend {blendOption=Linear}
    alphaScaleBiasClamp {bInterpResult=true, interpSpeedDecreasing=5, interpSpeedIncreasing=6}
    baseBoneRef {boneName=root}
    componentAppliedLinearAccClamp=<3000 3000 3000>
    componentLinearAccScale=<0.15 0.15 0.15>
    componentLinearVelScale=<0.15 0.15 0.15>
    overrideWorldGravity {z=-980}
   }
   n16 = "Constraint - dyn_shoulder_r (PPPPPP)"!(
    ComponentPose=@n15
    Alpha=1
    ConstraintWeights_0=0.0597
    ConstraintWeights_1:float=0.7266
    ConstraintWeights_2:float=0.063
    ConstraintWeights_3:float=0.0067
    ConstraintWeights_4:float=0.0011
    ConstraintWeights_5:float=0.1429
   ) -> (Pose) {
    alphaBoolBlend {blendOption=Linear}
    boneToModify {boneName=dyn_shoulder_r}
    constraintSetup [
     {targetBone {boneName=upperarm_r}}
     {targetBone {boneName=upperarm_twist_01_r}}
     {targetBone {boneName=upperarm_twist_02_r}}
     {targetBone {boneName=spine_03}}
     {targetBone {boneName=spine_05}}
     {targetBone {boneName=clavicle_r}}
    ]
    constraintWeights [1,1,1,1,1,1]
   }
   n17 = "Constraint - dyn_shoulder_l (PPPPPP)"!(
    ComponentPose=@n16
    Alpha=1
    ConstraintWeights_0=0.1989
    ConstraintWeights_1:float=0.0179
    ConstraintWeights_2:float=0.0172
    ConstraintWeights_3:float=0.0076
    ConstraintWeights_4:float=0.7377
    ConstraintWeights_5:float=0.0208
   ) -> (Pose) {
    alphaBoolBlend {blendOption=Linear}
    boneToModify {boneName=dyn_shoulder_l}
    constraintSetup [
     {targetBone {boneName=clavicle_l}}
     {targetBone {boneName=upperarm_twist_02_l}}
     {targetBone {boneName=spine_03}}
     {targetBone {boneName=spine_05}}
     {targetBone {boneName=upperarm_twist_01_l}}
     {targetBone {boneName=upperarm_l}}
    ]
    constraintWeights [1,1,1,1,1,1]
   }
   n18 = Anim Dynamics - Bone: dyn_shoulder_l(ComponentPose=@n17, Alpha=0.2, ExternalForce="0,0,0") {
    angularDampingOverride=0.99
    bGravityOverrideInSimSpace=true
    bOverrideAngularDamping=true
    bOverrideLinearDamping=true
    lODThreshold=1
    linearDampingOverride=0.99
    numSolverIterationsPostUpdate=4
    numSolverIterationsPreUpdate=16
    alphaBoolBlend {blendOption=Linear}
    boundBone {boneName=dyn_shoulder_l}
    physicsBodyDefinitions [{
      sphereCollisionRadius=0
      boundBone {boneName=dyn_shoulder_l}
      boxExtents=<10 20 10>
      constraintSetup {angularLimitsMax=<15 15 0>, angularLimitsMin=<-15 -15 0>, angularTarget=<0 0 0>, linearAxesMin=<-5 0 0>}
      localJointOffset=<10 0 -5>
     }]
   }
   n19 = Anim Dynamics - Bone: dyn_shoulder_r(ComponentPose=@n18, Alpha=0.2, ExternalForce="0,0,0") {
    angularDampingOverride=0.99
    bGravityOverrideInSimSpace=true
    bOverrideAngularDamping=true
    bOverrideLinearDamping=true
    lODThreshold=1
    linearDampingOverride=0.99
    numSolverIterationsPostUpdate=4
    numSolverIterationsPreUpdate=16
    alphaBoolBlend {blendOption=Linear}
    boundBone {boneName=dyn_shoulder_r}
    physicsBodyDefinitions [{
      sphereCollisionRadius=0
      boundBone {boneName=dyn_shoulder_r}
      boxExtents=<10 20 10>
      constraintSetup {angularLimitsMax=<15 15 0>, angularLimitsMin=<-15 -15 0>, angularTarget=<0 0 0>, linearAxesMin=<-5 0 0>}
      localJointOffset=<-10 0 3.5>
     }]
   }
   n20 = Anim Dynamics - Bone: dyn_scarf(ComponentPose=@n19, Alpha=1, ExternalForce="0,0,0") {
    angularDampingOverride=0.999
    bGravityOverrideInSimSpace=true
    bOverrideAngularDamping=true
    bOverrideLinearDamping=true
    linearDampingOverride=0.999
    alphaBoolBlend {blendOption=Linear}
    boundBone {boneName=dyn_scarf}
    componentAppliedLinearAccClamp=<6000 6000 6000>
    componentLinearAccScale=<0.25 0.25 0.25>
    componentLinearVelScale=<0.25 0.25 0.25>
    physicsBodyDefinitions [{
      sphereCollisionRadius=0
      boundBone {boneName=dyn_scarf}
      boxExtents=<20 5 15>
      constraintSetup {angularLimitsMax=<0 40 5>, angularLimitsMin=<0 -40 -5>, angularTarget=<0 0 0>}
      localJointOffset=<0 0 -5>
     }]
    relativeSpaceBone {boneName=spine_05}
   }
   n21 = "Constraint - dyn_feather_side_l_02 (PP)"!(ComponentPose=@n20, Alpha=1, ConstraintWeights_0=0.3063, ConstraintWeights_1:float=0.6937) -> (Pose) {
    alphaBoolBlend {blendOption=Linear}
    boneToModify {boneName=dyn_feather_side_l_02}
    constraintSetup [{targetBone {boneName=pelvis}}, {targetBone {boneName=thigh_twist_01_l}}]
    constraintWeights [1,1]
   }
   n22 = "Constraint - dyn_feather_side_l_01 (PP)"!(ComponentPose=@n21, Alpha=1, ConstraintWeights_0=0.0208, ConstraintWeights_1:float=0.9792) -> (Pose) {
    alphaBoolBlend {blendOption=Linear}
    boneToModify {boneName=dyn_feather_side_l_01}
    constraintSetup [{targetBone {boneName=pelvis}}, {targetBone {boneName=thigh_twist_01_l}}]
    constraintWeights [1,1]
   }
   n23 = "Constraint - dyn_feather_side_r_01 (PPP)"!(ComponentPose=@n22, Alpha=1, ConstraintWeights_0=0.5066, ConstraintWeights_1:float=0.4933, ConstraintWeights_2:float=0.0002) -> (Pose) {
    alphaBoolBlend {blendOption=Linear}
    boneToModify {boneName=dyn_feather_side_r_01}
    constraintSetup [{targetBone {boneName=pelvis}}, {targetBone {boneName=thigh_twist_01_r}}, {targetBone {boneName=spine_01}}]
    constraintWeights [1,1,1]
   }
   n24 = Anim Dynamics - Bone: dyn_feather_side_l_02(ComponentPose=@n23, Alpha=0.9, ExternalForce="0,0,0") {
    angularDampingOverride=0.9999
    bGravityOverrideInSimSpace=true
    bOverrideAngularDamping=true
    bOverrideLinearDamping=true
    bUseGravityOverride=true
    linearDampingOverride=0.9999
    alphaBoolBlend {blendOption=Linear}
    alphaScaleBiasClamp {bInterpResult=true, interpSpeedDecreasing=5, interpSpeedIncreasing=6}
    boundBone {boneName=dyn_feather_side_l_02}
    componentAppliedLinearAccClamp=<3000 3000 3000>
    componentLinearAccScale=<0.25 0.25 0.25>
    componentLinearVelScale=<0.25 0.25 0.25>
    gravityOverride {z=0.9}
    physicsBodyDefinitions [{
      sphereCollisionRadius=0
      boundBone {boneName=dyn_feather_side_l_02}
      boxExtents=<10 5 3>
      constraintSetup {angularLimitsMax=<0 60 5>, angularLimitsMin=<0 -5 -5>, angularTarget=<0 0 0>}
      localJointOffset=<3 0 0>
     }]
    relativeSpaceBone {boneName=root}
   }
   n25 = Anim Dynamics - Bone: dyn_feather_side_l_01(ComponentPose=@n24, Alpha=0.9, ExternalForce="0,0,0") {
    angularDampingOverride=0.9999
    bGravityOverrideInSimSpace=true
    bOverrideAngularDamping=true
    bOverrideLinearDamping=true
    bUseGravityOverride=true
    linearDampingOverride=0.9999
    alphaBoolBlend {blendOption=Linear}
    alphaScaleBiasClamp {bInterpResult=true, interpSpeedDecreasing=5, interpSpeedIncreasing=6}
    boundBone {boneName=dyn_feather_side_l_01}
    componentAppliedLinearAccClamp=<3000 3000 3000>
    componentLinearAccScale=<0.25 0.25 0.15>
    componentLinearVelScale=<0.25 0.25 0.15>
    gravityOverride {z=0.95}
    physicsBodyDefinitions [{
      sphereCollisionRadius=0
      boundBone {boneName=dyn_feather_side_l_01}
      boxExtents=<10 5 3>
      constraintSetup {angularLimitsMax=<0 30 5>, angularLimitsMin=<0 -10 -5>, angularTarget=<0 0 0>}
      localJointOffset=<3 0 0>
     }]
    relativeSpaceBone {boneName=root}
   }
   n26 = Anim Dynamics - Bone: dyn_feather_side_r_01(ComponentPose=@n25, Alpha=0.9, ExternalForce="0,0,0") {
    angularDampingOverride=0.9999
    bGravityOverrideInSimSpace=true
    bOverrideAngularDamping=true
    bOverrideLinearDamping=true
    bUseGravityOverride=true
    linearDampingOverride=0.9999
    alphaBoolBlend {blendOption=Linear}
    boundBone {boneName=dyn_feather_side_r_01}
    componentAppliedLinearAccClamp=<3000 3000 3000>
    componentLinearAccScale=<0.25 0.25 0.25>
    componentLinearVelScale=<0.25 0.25 0.25>
    gravityOverride {z=0.9}
    physicsBodyDefinitions [{
      sphereCollisionRadius=0
      boundBone {boneName=dyn_feather_side_r_01}
      boxExtents=<10 5 3>
      constraintSetup {angularLimitsMax=<0 30 5>, angularLimitsMin=<0 -10 -5>, angularTarget=<0 0 0>}
      localJointOffset=<3 0 0>
     }]
    relativeSpaceBone {boneName=root}
   }
   n27 = "Transform (Modify) Bone - Bone: skirt_r_deform_01"(ComponentPose=@n26, Alpha=@n1, Translation="-0.437,-1.035,1.304", Rotation="37.515,-15.510,-38.537", Scale="1,1,1") {
    rotationMode=BMM_Additive
    rotationSpace=BCS_BoneSpace
    translationMode=BMM_Additive
    translationSpace=BCS_BoneSpace
    alphaBoolBlend {blendOption=Linear}
    alphaScaleBiasClamp {bInterpResult=true}
    boneToModify {boneName=skirt_r_deform_01}
    rotation=<rot 37.515 -15.51 -38.537>
    translation=<-0.437 -1.035 1.304>
   }
   n28 = "Transform (Modify) Bone - Bone: dyn_skirt_r_front_base"(ComponentPose=@n27, Alpha=@n1, Translation="6.477,-3.942,5.933", Rotation="0.041,79.736,4.419", Scale="1,1,1") {
    rotationMode=BMM_Additive
    rotationSpace=BCS_BoneSpace
    translationMode=BMM_Additive
    translationSpace=BCS_BoneSpace
    alphaBoolBlend {blendOption=Linear}
    alphaScaleBiasClamp {bInterpResult=true}
    boneToModify {boneName=dyn_skirt_r_front_base}
    rotation=<rot 0.041 79.736 4.419>
    translation=<6.477 -3.942 5.933>
   }
   n29 = "Transform (Modify) Bone - Bone: dyn_skirt_r_back_base_parent"(ComponentPose=@n28, Alpha=@n1, Translation="-2.048,2.200,2.584", Rotation="38.157,-0.098,-0.054", Scale="1,1,1") {
    rotationMode=BMM_Additive
    rotationSpace=BCS_BoneSpace
    translationMode=BMM_Additive
    translationSpace=BCS_BoneSpace
    alphaBoolBlend {blendOption=Linear}
    alphaScaleBiasClamp {bInterpResult=true}
    boneToModify {boneName=dyn_skirt_r_back_base_parent}
    rotation=<rot 38.157 -0.098 -0.054>
    translation=<-2.048 2.2 2.584>
   }
   n30 = "Transform (Modify) Bone - Bone: dyn_skirt_r_back_01"(ComponentPose=@n29, Alpha=@n1, Translation="0,0,0", Rotation="54.062,-14.047,-11.332", Scale="1,1,1") {
    rotationMode=BMM_Additive
    rotationSpace=BCS_BoneSpace
    translationMode=BMM_Additive
    translationSpace=BCS_BoneSpace
    alphaBoolBlend {blendOption=Linear}
    alphaScaleBiasClamp {bInterpResult=true}
    boneToModify {boneName=dyn_skirt_r_back_01}
    rotation=<rot 54.062 -14.047 -11.332>
   }
   n31 = "Transform (Modify) Bone - Bone: skirt_r_deform_01"(ComponentPose=@n30, Alpha=Get CrouchWalk() -> (CrouchWalk:double), Translation="0,0,0", Rotation="15,0,-5", Scale="1,1,1") {
    rotationMode=BMM_Additive
    rotationSpace=BCS_BoneSpace
    translationMode=BMM_Additive
    translationSpace=BCS_BoneSpace
    alphaBoolBlend {blendOption=Linear}
    alphaScaleBiasClamp {bInterpResult=true, interpSpeedDecreasing=5, interpSpeedIncreasing=5}
    boneToModify {boneName=skirt_r_deform_01}
    rotation {pitch=15, roll=-5}
   }
   n32 = "Transform (Modify) Bone - Bone: dyn_skirt_r_back_base"(ComponentPose=@n31, Alpha=Get Jump_Alpha() -> (Jump_Alpha:double), Translation="0,4,-5", Rotation="0,-60,0", Scale="1,1,1") {
    rotationMode=BMM_Additive
    rotationSpace=BCS_BoneSpace
    translationMode=BMM_Additive
    translationSpace=BCS_BoneSpace
    alphaBoolBlend {blendOption=Linear}
    alphaScaleBiasClamp {bInterpResult=true, interpSpeedDecreasing=0.5, interpSpeedIncreasing=25}
    boneToModify {boneName=dyn_skirt_r_back_base}
    rotation {yaw=-60}
    translation {y=4, z=-5}
   }
   Output Pose(Result=Component To Local(ComponentPose=@n32)) {layerGroup=None}
   Save cached pose 'ADSCache'(Pose=Copy Pose From Mesh() {bUseAttachedParent=true}) {cacheName=ADSCache}
  ]}
 EventGraph {nodes [
   Event BlueprintUpdateAnimation()
   > n1 = Sequence!(execute) -> (then_0, then_1, then_2:Exec, then_3:Exec)
   n2 = GetDeltaTransformFromRefPose(self=GetOwnerSkeletalMeshComponent(), BoneName=head, BaseName=spine_05)
   > Set Head_Spine_Rot(Head_Spine_Rot:double=BreakRotator(InRot=BreakTransform(InTransform=@n2).Rotation).Yaw) -> (Output_Get:double)
   n3 = "float > float"(A=Get PawnSpeed2D(), B=0) {is=title K2Node_CallFunction}
   n4 = Get bIsJumping()
   Set Jump_Alpha(execute=@n1.then_3, Jump_Alpha:double=SelectFloat(A=1, B=0, bPickA=AND Boolean(A=@n4, B=@n3))) -> (Output_Get:double)
   Set Head_Rot_Back(execute=@n1.then_1, Head_Rot_Back:double=MapRangeClamped(Value=Get Head_Spine_Rot() -> (Head_Spine_Rot:double), InRangeA=0, InRangeB=-60)) -> (Output_Get:double)
   n5 = "float < float"!(A:double=Get PawnSpeed2D(), B:double=1) -> (ReturnValue:bool) {is=title K2Node_CallFunction}
   n6 = Get bIsCrouching()
   Set Crouch_Alpha(execute=@n1.then_2, Crouch_Alpha:double=SelectFloat(A=1, B=0, bPickA=AND Boolean(A=@n6, B=@n5))) -> (Output_Get:double)
   n7 = Get bIsCrouching()
   n8 = Get bIsMoving2D()
   > Set CrouchWalk(CrouchWalk:double="To Float (Boolean)"(InBool=AND Boolean(A=@n7, B=@n8))) -> (Output_Get:double)
   "float < float"(A=0, B=1)
  ]}
]
variables [
 Crouch_Alpha {type=number}
 CrouchWalk {type=number}
 Head_Rot_Back {type=number}
 Head_Spine_Rot {type=number}
 Jump_Alpha {type=number}
]
